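# Niagara explosion effect for the BattleBusBeam FX set, with a User.Stage parameter and, by its name, covering stages 1–3.
# Explode_Sparkles fires a delayed GPU sprite burst whose count follows Stage, shaped by curl noise, drag, gravity and speed-based stretching.
# Minimal004 adds a huge, high-offset pink light that swells and fades by particle age.
# The glow, wave, dark, spark and telegraph emitters list no modules or renderers.
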NiagaraSystem /GiftedTraveler/Assets/FX/BattleBusBeam/NS_Explosion_Stage123
dependencies {
 usedDataInterfaces [&NiagaraDataInterfaceCurve, &NiagaraDataInterfaceColorCurve]
 usedDynamicInputs [
  &Multiply_Vector2D
  &Waveform
  &MergeHelper_Vector2Switch
  &MakeFloat4FromVectorAndFloat
  &Divide_Float
  &Subtract_Float
  &Lerp_Float
  &MakeCustomIntFromBool
  &MakeCustomLinearColorFromBool
  &MergeHelper_VectorSwitch
  &MergeHelper_ColorSwitch
  &MakeCustomFloatFromBool
  &ColorFromCurve
 ]
 usedModules [
  &SystemState
  &self:SystemSpawnScript.SystemScriptSource.SystemScriptGraph.NiagaraNodeAssignment_0.SetVariables_#1
  &self:SystemSpawnScript.SystemScriptSource.SystemScriptGraph.NiagaraNodeAssignment_1.SetVariables_#2
  &self:SystemSpawnScript.SystemScriptSource.SystemScriptGraph.NiagaraNodeAssignment_2.SetVariables_#3
  &self:SystemSpawnScript.SystemScriptSource.SystemScriptGraph.NiagaraNodeAssignment_3.SetVariables_#4
  &self:SystemSpawnScript.SystemScriptSource.SystemScriptGraph.NiagaraNodeAssignment_4.SetVariables_#5
  &self:SystemSpawnScript.SystemScriptSource.SystemScriptGraph.NiagaraNodeAssignment_5.SetVariables_#6
  &self:SystemSpawnScript.SystemScriptSource.SystemScriptGraph.NiagaraNodeAssignment_6.SetVariables_#7
  &self:SystemSpawnScript.SystemScriptSource.SystemScriptGraph.NiagaraNodeAssignment_10.SetVariables_#8
  &self:SystemSpawnScript.SystemScriptSource.SystemScriptGraph.NiagaraNodeAssignment_11.SetVariables_#9
  &self:SystemSpawnScript.SystemScriptSource.SystemScriptGraph.NiagaraNodeAssignment_12.SetVariables_#10
  &self:SystemSpawnScript.SystemScriptSource.SystemScriptGraph.NiagaraNodeAssignment_7.SetVariables_#11
  &self:SystemSpawnScript.SystemScriptSource.SystemScriptGraph.NiagaraNodeAssignment_8.SetVariables_#12
 ]
}
emitters {
 EXP_BigGlow {rendererClasses [], renderers []}
 EXP_BrightWave {rendererClasses [], renderers []}
 EXP_Dark {rendererClasses [], renderers []}
 EXP_DarkWave {rendererClasses [], renderers []}
 EXP_Spark {rendererClasses [], renderers []}
 Explode_Sparkles {
  simTarget=GPUComputeSim
  emitterUpdateScript [
   &EmitterState(Life Cycle Mode=::NewEnumerator1, Loop Behavior=::NewEnumerator1, Loop Delay=2, Loop Duration=1, UseLoopDelay=true)
   &SpawnBurst_Instantaneous(
    Spawn Count=&MergeHelper_IntSwitch(0=0, 1=30, 2=100, 3=?, 4=?, 5=?, Index=$User.Stage, Num=::NewEnumerator1)
    Use Spawn Probability=false
    UseLoopCountLimit=false
   )
  ]
  particleSpawnScript [
   &InitializeParticle(
    Color=<rgba 2 0.407 0.7816 1>
    Lifetime=&RandomRangeFloat(Minimum=1.5, Maximum=2.5)
    Sprite Rotation Mode=::NewEnumerator0
    Sprite Size Max=<1 15>
    Sprite Size Min=<0.75 10>
    Sprite Size Mode=::NewEnumerator2
    UsePositionOffset=false
    Write Material Random=true
   )
   &ShapeLocation(MaskSpawnGroup=false, Override Seed=false, Rotation Mode=::NewEnumerator1, UseOwnerScale=true)
   &AddVelocity(
    Module.Use Falloff Distance=false
    Origin Offset=<0 0 -20>
    Override Seed=false
    Velocity Mode=::NewEnumerator1
    Velocity Speed=&RandomRangeFloat(Minimum=69088.1328, Maximum=104817.2344)
   )
   /SetVariables(Particles.Velocity=&Multiply_Vector(A=$Particles.Velocity, B=<1 1 0.5>)) {isSetParameters=true}
  ]
  particleUpdateScript [
   /SetVariables(Particles.Glow03_Color=&MakeLinearColorFromVectorAndFloat(
     "Vector (RGB)"=&MakeVectorFromLinearColorRGB(Color=&MultiplyLinearColor(Linear Color=<rgba 2 0 0.4704 0.2>, Scale Factor=&MakeVector4_FromFloat(Float=1)))
     "Float (Alpha)"=&Multiply_Float(
      A=0.4
      B=&FloatFromCurve(FloatCurve=DataInterface() {propertyValues {Curve {keys [
           {tangentMode=RCTM_User, time=0, value=1}
           {arriveTangent=-0.0405, interpMode=RCIM_Cubic, leaveTangent=-0.0405, tangentMode=RCTM_User, time=0.7009, value=1}
           {arriveTangent=-8.0089, interpMode=RCIM_Cubic, leaveTangent=-8.0089, tangentMode=RCTM_User, time=1, value=0}
          ]}}})
     )
    )) {isSetParameters=true}
   /SetVariables(Particles.Glow03_Size=&Multiply_Vector2D_ByFloat(
     Float=&FloatFromCurve(
      FloatCurve=DataInterface() {propertyValues {Curve {keys [
          <at 0 0 RCIM_Cubic>
          <at 0.2 1.3 RCIM_Cubic>
          {arriveTangent=-0.1423, interpMode=RCIM_Cubic, leaveTangent=-0.1423, tangentMode=RCTM_User, time=1, value=0}
         ]}}}
      Scale Curve=&Multiply_Float(A=1000, B=1)
     )
     Vector2D=<1 6>
    )) {isSetParameters=true}
   &ParticleState()
   &CurlNoiseForce(Noise Strength=&Multiply_Float(A=500, B=1), Pan Noise=true, Pan Noise Field=<0 0 1>)
   &ScaleSpriteSize(Scale Factor=&Vector2DFromFloat(Value=&FloatFromCurve(
      FloatCurve=DataInterface() {propertyValues {Curve {keys [
          <at 0 0 RCIM_Cubic>
          {interpMode=RCIM_Cubic, tangentMode=RCTM_User, time=0.2, value=1.2}
          {interpMode=RCIM_Cubic, leaveTangent=-2.7132, time=1, value=0}
         ]}}}
      Scale Curve=&Multiply_Float(A=0.7, B=100)
     )))
   &ScaleColor(
    Scale Alpha=&Multiply_Float(A=1, B=1)
    Scale RGB=&Multiply_VectorByFloat(Vector=<2 2 2>, Float=1)
    ScaleA=true
    ScaleRGB=true
   )
   &Drag(Drag=&RandomRangeFloat(Minimum=1, Maximum=3), Use Linear Drag=true, Use Rotational Drag=true)
   &GravityForce(Gravity=<0 0 -300>)
   &SolveForcesAndVelocity(Clamp Velocity=false, Limit Acceleration=false)
   &ScaleSpriteSizeBySpeed(Max Scale Factor=<1 3>, Min Scale Factor=<1 0.5>)
  ]
  rendererClasses [&NiagaraSpriteRendererProperties]
  renderers [
   {rendererClass=&NiagaraSpriteRendererProperties}
   {rendererClass=&NiagaraSpriteRendererProperties, rendererIndex=1}
  ]
 }
 GroundGlow {rendererClasses [], renderers []}
 GroundGlow001 {rendererClasses [], renderers []}
 GroundGlow003 {rendererClasses [], renderers []}
 Minimal004 {
  emitterUpdateScript [
   &EmitterState(Life Cycle Mode=::NewEnumerator1, Loop Behavior=::NewEnumerator1, Loop Delay=1.4, Loop Duration=1, UseLoopDelay=true)
   &SpawnBurst_Instantaneous(Use Spawn Probability=false, UseLoopCountLimit=false)
  ]
  particleSpawnScript [&InitializeParticle(Color=<rgba 20 1 6.2596 1>, Lifetime=5, Position Offset=<0 0 10000>, UsePositionOffset=true, Write Material Random=true)]
  particleUpdateScript [
   &ParticleState()
   &FN_Light_Attributes(Create Radius Attribute=true, Radius=40159.0117)
   &ScaleColor(
    Scale RGB=&VectorFromFloat(Value=&FloatFromCurve(
      FloatCurve=DataInterface() {propertyValues {Curve {keys [
          {arriveTangent=1.224, interpMode=RCIM_Cubic, leaveTangent=1.224, tangentMode=RCTM_User, time=0, value=0}
          {arriveTangent=0.7791, interpMode=RCIM_Cubic, leaveTangent=0.7791, tangentMode=RCTM_User, time=1.3635, value=0.625}
          {arriveTangent=-0.2526, interpMode=RCIM_Cubic, leaveTangent=-0.2526, time=1.6845, value=2.125}
          {interpMode=RCIM_Cubic, leaveTangent=0.3846, time=3.8381, value=0}
         ]}}}
      CurveIndex=$Particles.Age
     ))
    ScaleA=true
    ScaleRGB=true
   )
  ]
  rendererClasses [&NiagaraLightRendererProperties]
  renderers [{rendererClass=&NiagaraLightRendererProperties, rendererIndex=0}]
 }
 Telegraph_Energy {rendererClasses [], renderers []}
}
summary {
 systemName=NS_Explosion_Stage123
 emitters [
  {emitterName=GroundGlow, rendererClasses []}
  {emitterName=GroundGlow003, rendererClasses []}
  {emitterName=EXP_Dark, rendererClasses []}
  {emitterName=GroundGlow001, rendererClasses []}
  {emitterName=Explode_Sparkles, simTarget=GPUComputeSim, rendererClasses [&NiagaraSpriteRendererProperties]}
  {emitterName=Telegraph_Energy, rendererClasses []}
  {emitterName=EXP_BrightWave, rendererClasses []}
  {emitterName=EXP_BigGlow, rendererClasses []}
  {emitterName=EXP_Spark, rendererClasses []}
  {emitterName=EXP_DarkWave, rendererClasses []}
  {emitterName=Minimal004, rendererClasses [&NiagaraLightRendererProperties]}
 ]
 userVariables [
  User.bIsFrontEndPreview {type=NiagaraBool}
  User.ElementSizeScale {type=NiagaraFloat, defaultValue=45, description=""}
  User.ExplosionScale {type=NiagaraFloat, defaultValue=1, description=""}
  User.Return {type=NiagaraBool}
  User.SmokeSizeScale {type=NiagaraFloat, defaultValue=1, description=""}
  User.Stage {type=NiagaraInt32, defaultValue=1}
  User.Stage4 {type=NiagaraBool}
  User.TopFXDealy {type=NiagaraFloat, defaultValue=5.2, description=""}
  User.VelocityScale {type=NiagaraFloat, defaultValue=1, description=""}
 ]
}
