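# Niagara system for stages 1 to 3 of a Battle Bus beam effect, with stage user variables and system-level timeline and switch logic.
# Only AmbientSmoke is detailed here; the ray, glow and wave emitters list no renderers.
# Its burst and continuous sprites use separate spawn-group settings, with spawn probability from a system variable.
# Point and line attraction plus late-life upward gravity pull smoke into the column, while dynamic material parameters carry energy color.
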NiagaraSystem /GiftedTraveler/Assets/FX/BattleBusBeam/S34NE_Beam_Stage123
dependencies {
 usedDataInterfaces [&NiagaraDataInterfaceCurve, &NiagaraDataInterfaceColorCurve]
 usedDynamicInputs [
  &SetBoolByFloatComparison
  &Add_Float
  &Subtract_Float
  &TransformPosition
  &MakeVector
  &MakeFloatFromVector4
  &MakeLinearColorFromVectorAndFloat
  &Multiply_Float
  &Waveform
  &Multiply_Vector2D_ByFloat
  &MergeHelper_VectorSwitch
  &MergeHelper_Vector2Switch
  &MergeHelper_ColorSwitch
  &MergeHelper_FloatSwitch
 ]
 usedModules [
  &SystemState
  &Timeline
  &self:SystemSpawnScript.SystemScriptSource.SystemScriptGraph.NiagaraNodeAssignment_0.SetVariables_#1
  &self:SystemSpawnScript.SystemScriptSource.SystemScriptGraph.NiagaraNodeAssignment_1.SetVariables_#2
  &self:SystemSpawnScript.SystemScriptSource.SystemScriptGraph.NiagaraNodeAssignment_3.SetVariables_#3
  &self:SystemSpawnScript.SystemScriptSource.SystemScriptGraph.NiagaraNodeAssignment_9.SetVariables_#4
  &self:SystemSpawnScript.SystemScriptSource.SystemScriptGraph.NiagaraNodeAssignment_10.SetVariables_#5
  &self:SystemSpawnScript.SystemScriptSource.SystemScriptGraph.NiagaraNodeAssignment_4.SetVariables_#6
  &self:SystemSpawnScript.SystemScriptSource.SystemScriptGraph.NiagaraNodeAssignment_2.SetVariables_#7
  &self:SystemSpawnScript.SystemScriptSource.SystemScriptGraph.NiagaraNodeAssignment_5.SetVariables_#8
  &self:SystemSpawnScript.SystemScriptSource.SystemScriptGraph.NiagaraNodeAssignment_6.SetVariables_#9
  &self:SystemSpawnScript.SystemScriptSource.SystemScriptGraph.NiagaraNodeAssignment_7.SetVariables_#10
  &self:SystemSpawnScript.SystemScriptSource.SystemScriptGraph.NiagaraNodeAssignment_8.SetVariables_#11
  &self:SystemSpawnScript.SystemScriptSource.SystemScriptGraph.NiagaraNodeAssignment_11.SetVariables_#12
  &self:SystemSpawnScript.SystemScriptSource.SystemScriptGraph.NiagaraNodeAssignment_12.SetVariables_#13
  &self:SystemSpawnScript.SystemScriptSource.SystemScriptGraph.NiagaraNodeAssignment_13.SetVariables_#14
  &self:SystemSpawnScript.SystemScriptSource.SystemScriptGraph.NiagaraNodeAssignment_14.SetVariables_#15
  &self:SystemSpawnScript.SystemScriptSource.SystemScriptGraph.NiagaraNodeAssignment_15.SetVariables_#16
  &self:SystemSpawnScript.SystemScriptSource.SystemScriptGraph.NiagaraNodeAssignment_17.SetVariables_#17
  &self:SystemSpawnScript.SystemScriptSource.SystemScriptGraph.NiagaraNodeAssignment_18.SetVariables_#18
  &self:SystemSpawnScript.SystemScriptSource.SystemScriptGraph.NiagaraNodeAssignment_19.SetVariables_#19
  &self:SystemSpawnScript.SystemScriptSource.SystemScriptGraph.NiagaraNodeAssignment_20.SetVariables_#20
  &self:SystemSpawnScript.SystemScriptSource.SystemScriptGraph.NiagaraNodeAssignment_21.SetVariables_#21
 ]
}
emitters {
 AmbientRay_C {rendererClasses [], renderers []}
 AmbientSmoke {
  emitterSpawnScript [/SetVariables(Emitter.LightingCenter:Vector4f=&MakeFloat4FromVectorAndFloat(
     "Vector (RGB)"=&TransformVector(
      Vector=&AddVectorToPosition(Position=$Engine.Emitter.SimulationPosition, Vector=<0 0 100>)
      Source Space=::Simulation
      Destination Space=::World
     )
     "Float (Alpha)"=1
    )) {isSetParameters=true}]
  emitterUpdateScript [
   &EmitterState(Life Cycle Mode=::NewEnumerator1, Loop Behavior=::NewEnumerator1, Loop Duration=1, UseLoopDelay=true)
   &SpawnBurst_Instantaneous(Spawn Count=5, Spawn Group=1, Spawn Probability=$System.MS_ScaleSmokeSpawnRate, Use Spawn Probability=true, UseLoopCountLimit=false)
   &SpawnRate(Spawn Probability=$System.MS_ScaleSmokeSpawnRate, Use Spawn Probability=true)
  ]
  particleSpawnScript [
   &InitializeParticle(
    Color=<rgba 0.7166 0.673 1 0.31>
    Lifetime=&MaskFloatBySpawnGroup(
     Masked Float=&RandomRangeFloat(Minimum=11, Maximum=13)
     Passthrough Float=&RandomRangeFloat(Minimum=4, Maximum=9)
    )
    Sprite Rotation Mode=::NewEnumerator2
    Sprite Size Mode=::NewEnumerator3
    Uniform Sprite Size=25000
    UsePositionOffset=true
    Write Material Random=true
   ) {module=InitializeParticle001}
   &ShapeLocation(
    Handle Radius:NiagaraFloat=&MaskFloatBySpawnGroup(Masked Float=1, Passthrough Float=3000)
    Large Radius=&MaskFloatBySpawnGroup(Masked Float=37448.7969, Passthrough Float=26365.3906)
    MaskSpawnGroup=false
    Offset=&MaskVectorBySpawnGroup(Masked Vector=<0 0 0>, Passthrough Vector=<0 0 2500>)
    Offset Mode=::NewEnumerator0
    Override Seed=false
    Shape Primitive=::NewEnumerator3
    Surface Distribution=1
    UseOwnerScale=true
   )
   &CameraOffset(Camera Offset Amount=5000)
   &AddVelocity(
    Module.Use Falloff Distance=false
    Override Seed=false
    Velocity Mode=::NewEnumerator1
    Velocity Speed=&MaskFloatBySpawnGroup(Masked Float=0, Passthrough Float=&RandomRangeFloat(Minimum=-5000, Maximum=-1500))
   )
  ]
  particleUpdateScript [
   &ParticleState()
   &Drag(Drag=0.5, Use Linear Drag=true, Use Rotational Drag=false)
   &GravityForce(Gravity=&Multiply_Vector(
     A=<0 0 1>
     B=&VectorFromFloat(Value=&FloatFromCurve(
       FloatCurve=DataInterface() {propertyValues {Curve {keys [<at 0 0 RCIM_Cubic>, {interpMode=RCIM_Cubic, leaveTangent=0.2239, tangentMode=RCTM_Break, time=0.5, value=0}, <at 0.9 1>]}}}
       Scale Curve=10000
      ))
    ))
   &PointAttractionForce(Attraction Radius=120648296, Attraction Strength=7, Kill Within Radius=false, Module.Use Falloff=true, UseAttractorOffset=false)
   /Niagara/Modules/Update/Forces/LineAttractionForce(
    Attraction Strength:NiagaraFloat=&FloatFromCurve(
     FloatCurve=DataInterface() {propertyValues {Curve {keys [<at 0 0 RCIM_Cubic>, {interpMode=RCIM_Cubic, leaveTangent=0.2239, tangentMode=RCTM_Break, time=0.7, value=0}, <at 0.9 1>]}}}
     Scale Curve=15000
    )
    Line End:Vector3f=<0 0 20000>
    Line Start:Vector3f=<0 0 -10000>
    Use Attraction Falloff=false
   )
   &SolveForcesAndVelocity(Clamp Velocity=false, Limit Acceleration=false)
   &ScaleSpriteSize(Scale Factor=&Vector2DFromFloat(Value=&MaskFloatBySpawnGroup(
      Masked Float=&FloatFromCurve(
       FloatCurve=DataInterface() {propertyValues {Curve {keys [{interpMode=RCIM_Cubic, leaveTangent=1.1992, tangentMode=RCTM_Break, time=0, value=0.25}, <at 0.8 1>]}}}
       Scale Curve=2
      )
      Passthrough Float=1.2
     )))
   /SetVariables(Particles.Color_ParticleEnergy:LinearColor=&ColorFromCurve(ColorCurve=DataInterface() {propertyValues {
       AlphaCurve {keys [<at 0 1 RCIM_Cubic>]}
       BlueCurve {keys [<at 0.2206 0.715>]}
       GreenCurve {keys [<at 0.2206 0.2867>]}
       RedCurve {keys [<at 0.2206 2>]}
      }})) {isSetParameters=true}
   &DynamicMaterialParameters(
    Index 0 Param 1=&FloatFromCurve(FloatCurve=DataInterface() {propertyValues {Curve {keys [{tangentMode=RCTM_User, time=0, value=0}, {tangentMode=RCTM_User, time=1, value=1}]}}})
    Index 0 Param 3=100
    Index 0 Param 4=50
    Index 1 Param 1=&MakeFloatFromVector(Vector=&MakeVectorFromLinearColorRGB(Color=$Particles.Color_ParticleEnergy), Channel=::NewEnumerator0)
    Index 1 Param 2=&MakeFloatFromVector(Vector=&MakeVectorFromLinearColorRGB(Color=$Particles.Color_ParticleEnergy), Channel=::NewEnumerator1)
    Index 1 Param 3=&MakeFloatFromVector(Vector=&MakeVectorFromLinearColorRGB(Color=$Particles.Color_ParticleEnergy), Channel=::NewEnumerator2)
    Index 1 Param 4=&FloatFromCurve(FloatCurve=DataInterface() {propertyValues {Curve {keys [{arriveTangent=1.6946, interpMode=RCIM_Cubic, leaveTangent=1.6946, tangentMode=RCTM_User, time=0.2, value=0}, <at 1 1 RCIM_Cubic>]}}})
    Index 2 Param 1=45962.5977
    Index 2 Param 2=45000
    Index 2 Param 3=15000
    Index 2 Param 4=1
    Module.Index0Param0Write=true
    Module.Index0Param1Write=true
    Module.Index0Param2Write=true
    Module.Index0Param3Write=true
    Module.Index1Param0Write=true
    Module.Index1Param1Write=true
    Module.Index1Param2Write=true
    Module.Index1Param3Write=true
    Module.Index2Param0Write=true
    Module.Index2Param1Write=true
    Module.Index2Param2Write=true
    Module.Index2Param3Write=true
    Write Parameter Index 1=true
    Write Parameter Index 2=true
   )
   &ScaleColor(
    Scale Alpha=&MaskFloatBySpawnGroup(
     Masked Float=&FloatFromCurve(
      FloatCurve=DataInterface() {propertyValues {Curve {keys [
          {interpMode=RCIM_Cubic, leaveTangent=7.3008, tangentMode=RCTM_Break, time=0, value=0}
          {interpMode=RCIM_Cubic, tangentMode=RCTM_Break, time=0.2, value=1}
          {interpMode=RCIM_Cubic, leaveTangent=-0.0057, tangentMode=RCTM_Break, time=0.7, value=1}
          <at 1 0 RCIM_Cubic>
         ]}}}
      Scale Curve=&RandomRangeFloat(Minimum=1, Maximum=1)
     )
     Passthrough Float=&FloatFromCurve(
      FloatCurve=DataInterface() {propertyValues {Curve {keys [
          {interpMode=RCIM_Cubic, tangentMode=RCTM_Break, time=0.2, value=1}
          {interpMode=RCIM_Cubic, leaveTangent=-0.0057, tangentMode=RCTM_Break, time=0.7, value=1}
          <at 1 0 RCIM_Cubic>
         ]}}}
      Scale Curve=&FloatFromCurve(FloatCurve=DataInterface() {propertyValues {Curve {keys [<at 0 0>,<at 1 1>]}}}, CurveIndex=$Engine.System.Age)
     )
    )
    ScaleA=true
    ScaleRGB=true
   )
  ]
  rendererClasses [&NiagaraSpriteRendererProperties]
  renderers [{rendererClass=&NiagaraSpriteRendererProperties}]
 }
 GroundGlow {rendererClasses [], renderers []}
 GroundGlow002 {rendererClasses [], renderers []}
 GroundGlow006 {rendererClasses [], renderers []}
 MainRay {rendererClasses [], renderers []}
 MainRay001 {rendererClasses [], renderers []}
 MainRay010 {rendererClasses [], renderers []}
 TipGlow {rendererClasses [], renderers []}
 Wave {rendererClasses [], renderers []}
}
summary {
 systemName=S34NE_Beam_Stage123
 emitters [
  {emitterName=MainRay, rendererClasses []}
  {emitterName=MainRay001, rendererClasses []}
  {emitterName=AmbientSmoke, rendererClasses [&NiagaraSpriteRendererProperties]}
  {emitterName=GroundGlow, rendererClasses []}
  {emitterName=AmbientRay_C, rendererClasses []}
  {emitterName=MainRay010, rendererClasses []}
  {emitterName=GroundGlow002, rendererClasses []}
  {emitterName=TipGlow, rendererClasses []}
  {emitterName=GroundGlow006, rendererClasses []}
  {emitterName=Wave, rendererClasses []}
 ]
 userVariables [
  User.Stage4 {type=NiagaraBool}
  User.StageIndex {type=NiagaraInt32, defaultValue=1}
  User.TopFXDealy {type=NiagaraFloat, defaultValue=4.9, description=""}
 ]
}
